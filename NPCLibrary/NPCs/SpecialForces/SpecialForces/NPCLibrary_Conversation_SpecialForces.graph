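# Conversation database for the Special Forces NPC, opened through the Conversation.Entry.SpecialForces entry tag.
# A single speech node gives a default military-style greeting, with contextual lines for a long gap since the last match, a first-ever meeting, or holding enough Bars.
# The speech then branches through link nodes into shared conversations for Season 17 Week 2, NPC library quests, and basic services.
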ConversationDatabase /NPCLibrary/NPCs/SpecialForces/SpecialForces/NPCLibrary_Conversation_SpecialForces
entry_points [{tag=Conversation.Entry.SpecialForces, destinations [#1]}]
nodes [
 ConversationEntryPointNode_7 {
  id=#1
  class=ConversationEntryPointNode
  outputs [#2]
  properties {entryTag {tagName=Conversation.Entry.SpecialForces}}
 }
 Speech_C_4 {
  id=#2
  class=Speech_C
  outputs [#3, #4, #5]
  properties {
   nodeName=""
   generalConfig {
    defaultMessage="Reporting for duty!"
    contextualMessages [
     {
      message="I've been awaiting your orders, <b>Captain</>. "
      contextRequirements [{requirement {class=FortControllerRequirement_CheckTimeSinceLastPlayedMatch}}]
      priority {curve {rowName=Tandem.Speech.Priority.WhileSincePlayed}}
     }
     {
      message=A new C.O.? At your service.
      contextRequirements [{requirement {class=FortControllerRequirement_IsFirstEverConversationWithNPC}}]
      priority {curve {rowName=Tandem.Speech.Priority.FirstEverInteraction}}
     }
     {
      message=Those Bars'll buy a lot of fresh tactical gear.
      contextRequirements [{requirement {class=FortControllerRequirement_HasBars, properties {thresholdTestConfig {requirement=GreaterThanOrEqual}}}}]
      priority {curve {rowName=None}}
     }
    ]
   }
  }
 }
 ConversationLinkNode_0 {id=#5, class=ConversationLinkNode, properties {remoteEntryTag {tagName=Conversation.Entry.BasicServices}}}
 ConversationLinkNode_1 {id=#4, class=ConversationLinkNode, properties {remoteEntryTag {tagName=Conversation.Entry.NPCLibrary.Quests}}}
 ConversationLinkNode_2 {id=#3, class=ConversationLinkNode, properties {remoteEntryTag {tagName=Conversation.Entry.Season17.Week2}}}
]
